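# Post-process-style body AnimBP for the F_MED_CowboyHat female cosmetic, copying its pose from the attached parent mesh.
# Each update caches the head mesh's jaw and under-chin transforms when face bones exist.
# bHasFaceBones selects between the copied pose and one replacing those bones' transforms.
# Weighted constraints then drive dynamic shoulder bones, feeding a Control Rig that receives bIsMoving2D, before an additive right-shoulder nudge while skydiving.
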AnimBlueprint /BRCosmetics/Characters/Player/Female/Medium/Bodies/F_MED_CowboyHat/Meshes/F_MED_CowboyHat_AnimBP
target_skeleton=&/F_MED_CowboyHat_Skeleton
graphs [
 AnimGraph {nodes [
   n1 = "Transform (Modify) Bone - Bone: FACIAL_C_Jaw"!(
    ComponentPose=Local To Component(LocalPose=Use cached pose 'InputPose'())
    bAlphaBoolEnabled:bool=True
    Translation="0,0,0"
    Rotation="0,0,20"
    Scale="1,1,1"
   ) -> (Pose) {
    alphaInputType=Bool
    rotationMode=BMM_Replace
    rotationSpace=BCS_ParentBoneSpace
    translationMode=BMM_Replace
    translationSpace=BCS_ParentBoneSpace
    alphaBoolBlend {blendOption=Linear}
    boneToModify {boneName=FACIAL_C_Jaw}
    rotation {roll=20}
   }
   n2 = "Transform (Modify) Bone - Bone: FACIAL_C_Jawline"!(ComponentPose=@n1, bAlphaBoolEnabled:bool=True, Translation="0,0,0", Rotation="0,0,20", Scale="1,1,1") -> (Pose) {
    alphaInputType=Bool
    rotationMode=BMM_Replace
    rotationSpace=BCS_ParentBoneSpace
    translationMode=BMM_Replace
    translationSpace=BCS_ParentBoneSpace
    alphaBoolBlend {blendOption=Linear}
    boneToModify {boneName=FACIAL_C_Jawline}
    rotation {roll=20}
   }
   n3 = "Transform (Modify) Bone - Bone: FACIAL_L_Jawline"!(ComponentPose=@n2, bAlphaBoolEnabled:bool=True, Translation="0,0,0", Rotation="0,0,20", Scale="1,1,1") -> (Pose) {
    alphaInputType=Bool
    rotationMode=BMM_Replace
    rotationSpace=BCS_ParentBoneSpace
    translationMode=BMM_Replace
    translationSpace=BCS_ParentBoneSpace
    alphaBoolBlend {blendOption=Linear}
    boneToModify {boneName=FACIAL_L_Jawline}
    rotation {roll=20}
   }
   n4 = "Transform (Modify) Bone - Bone: FACIAL_R_Jawline"!(ComponentPose=@n3, bAlphaBoolEnabled:bool=True, Translation="0,0,0", Rotation="0,0,20", Scale="1,1,1") -> (Pose) {
    alphaInputType=Bool
    rotationMode=BMM_Replace
    rotationSpace=BCS_ParentBoneSpace
    translationMode=BMM_Replace
    translationSpace=BCS_ParentBoneSpace
    alphaBoolBlend {blendOption=Linear}
    boneToModify {boneName=FACIAL_R_Jawline}
    rotation {roll=20}
   }
   n5 = "Transform (Modify) Bone - Bone: FACIAL_C_UnderChin"!(ComponentPose=@n4, bAlphaBoolEnabled:bool=True, Translation="0,0,0", Rotation="0,0,20", Scale="1,1,1") -> (Pose) {
    alphaInputType=Bool
    rotationMode=BMM_Replace
    rotationSpace=BCS_ParentBoneSpace
    translationMode=BMM_Replace
    translationSpace=BCS_ParentBoneSpace
    alphaBoolBlend {blendOption=Linear}
    boneToModify {boneName=FACIAL_C_UnderChin}
    rotation {roll=20}
   }
   n6 = "Transform (Modify) Bone - Bone: FACIAL_L_UnderChin"!(ComponentPose=@n5, bAlphaBoolEnabled:bool=True, Translation="0,0,0", Rotation="0,0,20", Scale="1,1,1") -> (Pose) {
    alphaInputType=Bool
    rotationMode=BMM_Replace
    rotationSpace=BCS_ParentBoneSpace
    translationMode=BMM_Replace
    translationSpace=BCS_ParentBoneSpace
    alphaBoolBlend {blendOption=Linear}
    boneToModify {boneName=FACIAL_L_UnderChin}
    rotation {roll=20}
   }
   n7 = "Transform (Modify) Bone - Bone: FACIAL_R_UnderChin"!(ComponentPose=@n6, bAlphaBoolEnabled:bool=True, Translation="0,0,0", Rotation="0,0,20", Scale="1,1,1") -> (Pose) {
    alphaInputType=Bool
    rotationMode=BMM_Replace
    rotationSpace=BCS_ParentBoneSpace
    translationMode=BMM_Replace
    translationSpace=BCS_ParentBoneSpace
    alphaBoolBlend {blendOption=Linear}
    boneToModify {boneName=FACIAL_R_UnderChin}
    rotation {roll=20}
   }
   n8 = Component To Local(ComponentPose=@n7)
   n9 = Use cached pose 'InputPose'()
   n10 = Get bHasFaceBones() -> (bHasFaceBones:bool)
   n11 = "Constraint - dyn_r_shoulder_master (PPPP)"!(
    ComponentPose=Local To Component(LocalPose=Blend Poses by bool(BlendPose_0=@n8, BlendPose_1=@n9, bActiveValue=@n10) {blendType=Linear})
    Alpha=1
    ConstraintWeights_0=0.0068
    ConstraintWeights_1:float=0.0195
    ConstraintWeights_2:float=0.5114
    ConstraintWeights_3:float=0.4623
   ) -> (Pose) {
    alphaBoolBlend {blendOption=Linear}
    boneToModify {boneName=dyn_r_shoulder_master}
    constraintSetup [
     {targetBone {boneName=spine_04}}
     {targetBone {boneName=spine_05}}
     {targetBone {boneName=clavicle_r}}
     {targetBone {boneName=upperarm_twist_01_r}}
    ]
    constraintWeights [1,1,1,1]
   }
   n12 = "Constraint - dyn_r_shoulder_fr_mstr (PPPP)"!(
    ComponentPose=@n11
    Alpha=1
    ConstraintWeights_0=0.1179
    ConstraintWeights_1:float=0.1739
    ConstraintWeights_2:float=0.4658
    ConstraintWeights_3:float=0.2424
   ) -> (Pose) {
    alphaBoolBlend {blendOption=Linear}
    boneToModify {boneName=dyn_r_shoulder_fr_mstr}
    constraintSetup [
     {targetBone {boneName=spine_04}}
     {targetBone {boneName=spine_05}}
     {targetBone {boneName=clavicle_r}}
     {targetBone {boneName=upperarm_twist_01_r}}
    ]
    constraintWeights [1,1,1,1]
   }
   n13 = "Constraint - dyn_l_shoulder_master (PPP)"!(ComponentPose=@n12, Alpha=1, ConstraintWeights_0=0.4954, ConstraintWeights_1:float=0.0195, ConstraintWeights_2:float=0.4852) -> (Pose) {
    alphaBoolBlend {blendOption=Linear}
    boneToModify {boneName=dyn_l_shoulder_master}
    constraintSetup [{targetBone {boneName=upperarm_twist_01_l}}, {targetBone {boneName=spine_05}}, {targetBone {boneName=clavicle_l}}]
    constraintWeights [1,1,1]
   }
   n14 = "Constraint - dyn_l_shoulder_fr_mstr (PPPP)"!(
    ComponentPose=@n13
    Alpha=1
    ConstraintWeights_0=0.208
    ConstraintWeights_1:float=0.1355
    ConstraintWeights_2:float=0.1843
    ConstraintWeights_3:float=0.4723
   ) -> (Pose) {
    alphaBoolBlend {blendOption=Linear}
    boneToModify {boneName=dyn_l_shoulder_fr_mstr}
    constraintSetup [
     {targetBone {boneName=upperarm_twist_01_l}}
     {targetBone {boneName=spine_04}}
     {targetBone {boneName=spine_05}}
     {targetBone {boneName=clavicle_l}}
    ]
    constraintWeights [1,1,1,1]
   }
   n15 = Control Rig!(Source=Component To Local(ComponentPose=@n14), Alpha, bIsMoving2D:bool=Get bIsMoving2D()) -> (Pose) {
    alphaBoolBlend {blendOption=Linear}
    controlRigAssetReference {blueprintRigClass=&/F_MED_CowboyHat_CtrlRig.F_MED_CowboyHat_CtrlRig_C}
   }
   n16 = "Transform (Modify) Bone - Bone: dyn_r_shoulder"!(
    ComponentPose=Local To Component(LocalPose=@n15)
    bAlphaBoolEnabled:bool=Get bIsSkyDiving()
    Translation="1,1,0"
    Rotation="0,0,0"
    Scale="1,1,1"
   ) -> (Pose) {
    alphaInputType=Bool
    rotationMode=BMM_Additive
    rotationSpace=BCS_BoneSpace
    translationMode=BMM_Additive
    translationSpace=BCS_BoneSpace
    alphaBoolBlend {blendInTime=1, blendOption=Linear, blendOutTime=1}
    boneToModify {boneName=dyn_r_shoulder}
    translation=<1 1>
   }
   Output Pose(Result=Component To Local(ComponentPose=@n16)) {layerGroup=None}
   Save cached pose 'InputPose'(Pose=Copy Pose From Mesh() {bUseAttachedParent=true}) {cacheName=InputPose}
  ]}
 EventGraph {nodes [
   n1 = Event OnCharacterCustomizationLoaded()
   > Set HeadMesh(HeadMesh:Skeletal Mesh Component Object Reference=Get Array Item with EFort Custom Part Type(Array=@n1.CustomMeshParts, PartType=Head)) -> (Output_Get:Skeletal Mesh Component Object Reference)
   Event BlueprintUpdateAnimation()
   > Is Valid(InputObject=Get HeadMesh() -> (HeadMesh:Skeletal Mesh Component Object Reference))
   n2 = GetSocketBoneName!(self:Skinned Mesh Component Object Reference=Get HeadMesh(), InSocketName:Name=FACIAL_C_FacialRoot) -> (ReturnValue:Name=None)
   > n3 = Set bHasFaceBones(bHasFaceBones:bool="Not Equal Exactly (String)"(A="To String (Name)"(InName=@n2), B=None)) -> (Output_Get:bool)
   > Branch(Condition=@n3)
   n4 = BreakTransform(InTransform=GetSocketTransform(self=Get HeadMesh(), InSocketName=FACIAL_C_UnderChin, TransformSpace=RTS_ParentBoneSpace))
   > Set FACIAL_C_UnderChin_Location(FACIAL_C_UnderChin_Location:Vector=@n4) -> (Output_Get:Vector)
   > Set FACIAL_C_UnderChin_Rotation(FACIAL_C_UnderChin_Rotation:Rotator=@n4.Rotation) -> (Output_Get:Rotator)
   n5 = BreakTransform(InTransform=GetSocketTransform(self=Get HeadMesh(), InSocketName=FACIAL_L_UnderChin, TransformSpace=RTS_ParentBoneSpace))
   > Set FACIAL_L_UnderChin_Location(FACIAL_L_UnderChin_Location:Vector=@n5) -> (Output_Get:Vector)
   > Set FACIAL_L_UnderChin_Rotation(FACIAL_L_UnderChin_Rotation:Rotator=@n5.Rotation) -> (Output_Get:Rotator)
   n6 = BreakTransform(InTransform=GetSocketTransform(self=Get HeadMesh(), InSocketName=FACIAL_R_UnderChin, TransformSpace=RTS_ParentBoneSpace))
   > Set FACIAL_R_UnderChin_Location(FACIAL_R_UnderChin_Location:Vector=@n6) -> (Output_Get:Vector)
   > Set FACIAL_R_UnderChin_Rotation(FACIAL_R_UnderChin_Rotation:Rotator=@n6.Rotation) -> (Output_Get:Rotator)
   n7 = BreakTransform(InTransform=GetSocketTransform(self=Get HeadMesh(), InSocketName=FACIAL_C_Jaw, TransformSpace=RTS_ParentBoneSpace))
   > Set FACIAL_C_Jaw_Location(FACIAL_C_Jaw_Location:Vector=@n7) -> (Output_Get:Vector)
   > Set FACIAL_C_Jaw_Rotation(FACIAL_C_Jaw_Rotation:Rotator=@n7.Rotation) -> (Output_Get:Rotator)
   n8 = BreakTransform(InTransform=GetSocketTransform(self=Get HeadMesh(), InSocketName=FACIAL_C_Jawline, TransformSpace=RTS_ParentBoneSpace))
   > Set FACIAL_C_Jawline_Location(FACIAL_C_Jawline_Location:Vector=@n8) -> (Output_Get:Vector)
   > Set FACIAL_C_Jawline_Rotation(FACIAL_C_Jawline_Rotation:Rotator=@n8.Rotation) -> (Output_Get:Rotator)
   n9 = BreakTransform(InTransform=GetSocketTransform(self=Get HeadMesh(), InSocketName=FACIAL_L_Jawline, TransformSpace=RTS_ParentBoneSpace))
   > Set FACIAL_L_Jawline_Location(FACIAL_L_Jawline_Location:Vector=@n9) -> (Output_Get:Vector)
   > Set FACIAL_L_Jawline_Rotation(FACIAL_L_Jawline_Rotation:Rotator=@n9.Rotation) -> (Output_Get:Rotator)
   n10 = BreakTransform(InTransform=GetSocketTransform(self=Get HeadMesh(), InSocketName=FACIAL_R_Jawline, TransformSpace=RTS_ParentBoneSpace))
   > Set FACIAL_R_Jawline_Location(FACIAL_R_Jawline_Location:Vector=@n10) -> (Output_Get:Vector)
   > Set FACIAL_R_Jawline_Rotation(FACIAL_R_Jawline_Rotation:Rotator=@n10.Rotation) -> (Output_Get:Rotator)
  ]}
]
variables [
 bHasFaceBones {type=boolean}
 FACIAL_C_Jaw_Location {type=Vector}
 FACIAL_C_Jaw_Rotation {type=Rotator}
 FACIAL_C_Jawline_Location {type=Vector}
 FACIAL_C_Jawline_Rotation {type=Rotator}
 FACIAL_C_UnderChin_Location {type=Vector}
 FACIAL_C_UnderChin_Rotation {type=Rotator}
 FACIAL_L_Jawline_Location {type=Vector}
 FACIAL_L_Jawline_Rotation {type=Rotator}
 FACIAL_L_UnderChin_Location {type=Vector}
 FACIAL_L_UnderChin_Rotation {type=Rotator}
 FACIAL_R_Jawline_Location {type=Vector}
 FACIAL_R_Jawline_Rotation {type=Rotator}
 FACIAL_R_UnderChin_Location {type=Vector}
 FACIAL_R_UnderChin_Rotation {type=Rotator}
 HeadMesh {type=SkeletalMeshComponent}
]
