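# Animation Blueprint for the female CowboyHat face accessory, posing it on top of the pose copied from its attached parent mesh.
# Each update caches head-mesh eye and cheek bone transforms and checks for face bones, which selects a pose with those bones overridden.
# The result feeds a Control Rig given a not-skydiving alpha, then fixed rotation and scale tweaks on four horn bones.
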AnimBlueprint /BRCosmetics/Characters/Player/Female/Medium/Bodies/F_MED_CowboyHat/Meshes/Parts/F_MED_CowboyHat_FaceAcc_AnimBP
target_skeleton=&/F_MED_CowboyHat_FaceAcc_Skeleton
graphs [
 AnimGraph {nodes [
   n1 = "Transform (Modify) Bone - Bone: FACIAL_L_EyeCornerOuter"!(
    ComponentPose=Local To Component(LocalPose=Use cached pose 'InputPose'())
    bAlphaBoolEnabled:bool=True
    Translation="0,0,0"
    Rotation="0,0,20"
    Scale="1,1,1"
   ) -> (Pose) {
    alphaInputType=Bool
    rotationMode=BMM_Replace
    rotationSpace=BCS_ParentBoneSpace
    translationMode=BMM_Replace
    translationSpace=BCS_ParentBoneSpace
    alphaBoolBlend {blendOption=Linear}
    boneToModify {boneName=FACIAL_L_EyeCornerOuter}
    rotation {roll=20}
   }
   n2 = "Transform (Modify) Bone - Bone: FACIAL_R_EyeCornerOuter"!(ComponentPose=@n1, bAlphaBoolEnabled:bool=True, Translation="0,0,0", Rotation="0,0,20", Scale="1,1,1") -> (Pose) {
    alphaInputType=Bool
    rotationMode=BMM_Replace
    rotationSpace=BCS_ParentBoneSpace
    translationMode=BMM_Replace
    translationSpace=BCS_ParentBoneSpace
    alphaBoolBlend {blendOption=Linear}
    boneToModify {boneName=FACIAL_R_EyeCornerOuter}
    rotation {roll=20}
   }
   n3 = "Transform (Modify) Bone - Bone: FACIAL_L_CheekInner"!(ComponentPose=@n2, bAlphaBoolEnabled:bool=True, Translation="0,0,0", Rotation="0,0,20", Scale="1,1,1") -> (Pose) {
    alphaInputType=Bool
    rotationMode=BMM_Replace
    rotationSpace=BCS_ParentBoneSpace
    translationMode=BMM_Replace
    translationSpace=BCS_ParentBoneSpace
    alphaBoolBlend {blendOption=Linear}
    boneToModify {boneName=FACIAL_L_CheekInner}
    rotation {roll=20}
   }
   n4 = "Transform (Modify) Bone - Bone: FACIAL_R_CheekInner"!(ComponentPose=@n3, bAlphaBoolEnabled:bool=True, Translation="0,0,0", Rotation="0,0,20", Scale="1,1,1") -> (Pose) {
    alphaInputType=Bool
    rotationMode=BMM_Replace
    rotationSpace=BCS_ParentBoneSpace
    translationMode=BMM_Replace
    translationSpace=BCS_ParentBoneSpace
    alphaBoolBlend {blendOption=Linear}
    boneToModify {boneName=FACIAL_R_CheekInner}
    rotation {roll=20}
   }
   n5 = "Transform (Modify) Bone - Bone: FACIAL_L_CheekOuter"!(ComponentPose=@n4, bAlphaBoolEnabled:bool=True, Translation="0,0,0", Rotation="0,0,20", Scale="1,1,1") -> (Pose) {
    alphaInputType=Bool
    rotationMode=BMM_Replace
    rotationSpace=BCS_ParentBoneSpace
    translationMode=BMM_Replace
    translationSpace=BCS_ParentBoneSpace
    alphaBoolBlend {blendOption=Linear}
    boneToModify {boneName=FACIAL_L_CheekOuter}
    rotation {roll=20}
   }
   n6 = "Transform (Modify) Bone - Bone: FACIAL_R_CheekOuter"!(ComponentPose=@n5, bAlphaBoolEnabled:bool=True, Translation="0,0,0", Rotation="0,0,20", Scale="1,1,1") -> (Pose) {
    alphaInputType=Bool
    rotationMode=BMM_Replace
    rotationSpace=BCS_ParentBoneSpace
    translationMode=BMM_Replace
    translationSpace=BCS_ParentBoneSpace
    alphaBoolBlend {blendOption=Linear}
    boneToModify {boneName=FACIAL_R_CheekOuter}
    rotation {roll=20}
   }
   n7 = "Transform (Modify) Bone - Bone: FACIAL_L_EyelidLowerB"!(ComponentPose=@n6, bAlphaBoolEnabled:bool=True, Translation="0,0,0", Rotation="0,0,20", Scale="1,1,1") -> (Pose) {
    alphaInputType=Bool
    rotationMode=BMM_Replace
    rotationSpace=BCS_ParentBoneSpace
    translationMode=BMM_Replace
    translationSpace=BCS_ParentBoneSpace
    alphaBoolBlend {blendOption=Linear}
    boneToModify {boneName=FACIAL_L_EyelidLowerB}
    rotation {roll=20}
   }
   n8 = "Transform (Modify) Bone - Bone: FACIAL_R_EyelidLowerB"!(ComponentPose=@n7, bAlphaBoolEnabled:bool=True, Translation="0,0,0", Rotation="0,0,20", Scale="1,1,1") -> (Pose) {
    alphaInputType=Bool
    rotationMode=BMM_Replace
    rotationSpace=BCS_ParentBoneSpace
    translationMode=BMM_Replace
    translationSpace=BCS_ParentBoneSpace
    alphaBoolBlend {blendOption=Linear}
    boneToModify {boneName=FACIAL_R_EyelidLowerB}
    rotation {roll=20}
   }
   n9 = "Transform (Modify) Bone - Bone: FACIAL_L_EyesackLower"!(ComponentPose=@n8, bAlphaBoolEnabled:bool=True, Translation="0,0,0", Rotation="0,0,20", Scale="1,1,1") -> (Pose) {
    alphaInputType=Bool
    rotationMode=BMM_Replace
    rotationSpace=BCS_ParentBoneSpace
    translationMode=BMM_Replace
    translationSpace=BCS_ParentBoneSpace
    alphaBoolBlend {blendOption=Linear}
    boneToModify {boneName=FACIAL_L_EyesackLower}
    rotation {roll=20}
   }
   n10 = "Transform (Modify) Bone - Bone: FACIAL_R_EyesackLower"!(ComponentPose=@n9, bAlphaBoolEnabled:bool=True, Translation="0,0,0", Rotation="0,0,20", Scale="1,1,1") -> (Pose) {
    alphaInputType=Bool
    rotationMode=BMM_Replace
    rotationSpace=BCS_ParentBoneSpace
    translationMode=BMM_Replace
    translationSpace=BCS_ParentBoneSpace
    alphaBoolBlend {blendOption=Linear}
    boneToModify {boneName=FACIAL_R_EyesackLower}
    rotation {roll=20}
   }
   n11 = Control Rig!(
    Source=Blend Poses by bool(
     BlendPose_0=Component To Local(ComponentPose=@n10)
     BlendPose_1=Use cached pose 'InputPose'()
     bActiveValue=Get bHasFaceBones() -> (bHasFaceBones:bool)
    ) {blendType=Linear}
    Alpha
    Is Not Skydiving Alpha:double=Get Is Not Skydiving Alpha() -> (Is Not Skydiving Alpha:double)
   ) -> (Pose) {
    alphaBoolBlend {blendOption=Linear}
    controlRigAssetReference {blueprintRigClass=&/F_MED_CowboyHat_FaceAcc_CtrlRig.F_MED_CowboyHat_FaceAcc_CtrlRig_C}
   }
   n12 = "Transform (Modify) Bone - Bone: dyn_horn_upper_r_01"!(ComponentPose=Local To Component(LocalPose=@n11), Alpha=1, Rotation="0,-10,0", Scale="0.9,1,1") -> (Pose) {
    rotationMode=BMM_Additive
    rotationSpace=BCS_BoneSpace
    scaleMode=BMM_Replace
    scaleSpace=BCS_BoneSpace
    alphaBoolBlend {blendInTime=0.4, blendOption=Linear, blendOutTime=0.35}
    boneToModify {boneName=dyn_horn_upper_r_01}
    rotation {yaw=-10}
    scale {x=0.9}
   }
   n13 = "Transform (Modify) Bone - Bone: dyn_horn_upper_r_02"!(ComponentPose=@n12, Alpha=1, Rotation="0,-10,0", Scale="0.9,1,1") -> (Pose) {
    rotationMode=BMM_Additive
    rotationSpace=BCS_BoneSpace
    scaleMode=BMM_Replace
    scaleSpace=BCS_BoneSpace
    alphaBoolBlend {blendInTime=0.4, blendOption=Linear, blendOutTime=0.35}
    boneToModify {boneName=dyn_horn_upper_r_02}
    rotation {yaw=-10}
    scale {x=0.9}
   }
   n14 = "Transform (Modify) Bone - Bone: dyn_horn_l"!(ComponentPose=@n13, Alpha=1, Rotation="0,-10,10", Scale="0.9,1,1") -> (Pose) {
    rotationMode=BMM_Additive
    rotationSpace=BCS_BoneSpace
    scaleMode=BMM_Replace
    scaleSpace=BCS_BoneSpace
    alphaBoolBlend {blendInTime=0.4, blendOption=Linear, blendOutTime=0.35}
    boneToModify {boneName=dyn_horn_l}
    rotation {roll=10, yaw=-10}
    scale {x=0.9}
   }
   n15 = "Transform (Modify) Bone - Bone: dyn_horn_r"!(ComponentPose=@n14, Alpha=1, Rotation="0,-10,-10", Scale="0.9,1,1") -> (Pose) {
    rotationMode=BMM_Additive
    rotationSpace=BCS_BoneSpace
    scaleMode=BMM_Replace
    scaleSpace=BCS_BoneSpace
    alphaBoolBlend {blendInTime=0.4, blendOption=Linear, blendOutTime=0.35}
    boneToModify {boneName=dyn_horn_r}
    rotation {roll=-10, yaw=-10}
    scale {x=0.9}
   }
   Output Pose(Result=Component To Local(ComponentPose=@n15)) {layerGroup=None}
   Save cached pose 'InputPose'(Pose=Copy Pose From Mesh() {bUseAttachedParent=true}) {cacheName=InputPose}
  ]}
 EventGraph {nodes [
   n1 = Event OnCharacterCustomizationLoaded()
   > Set HeadMesh(HeadMesh:Skeletal Mesh Component Object Reference=Get Array Item with EFort Custom Part Type(Array=@n1.CustomMeshParts, PartType=Head)) -> (Output_Get:Skeletal Mesh Component Object Reference)
   Event BlueprintUpdateAnimation()
   > n2 = Sequence!(execute) -> (then_0, then_1, then_2:Exec)
   > Set bIsMovingCrouched(bIsMovingCrouched:bool=AND Boolean!(A=Get bIsMoving2D(), B=Get bIsCrouching(), C:bool=Get bIsTargeting()) -> (ReturnValue)) -> (Output_Get:bool)
   > Is Valid(InputObject=Get HeadMesh() -> (HeadMesh:Skeletal Mesh Component Object Reference))
   n3 = GetSocketBoneName!(self:Skinned Mesh Component Object Reference=Get HeadMesh(), InSocketName:Name=FACIAL_C_FacialRoot) -> (ReturnValue:Name=None)
   > n4 = Set bHasFaceBones(bHasFaceBones:bool="Not Equal Exactly (String)"(A="To String (Name)"(InName=@n3), B=None)) -> (Output_Get:bool)
   > Branch(Condition=@n4)
   n5 = BreakTransform(InTransform=GetSocketTransform(self=Get HeadMesh(), InSocketName=FACIAL_L_EyesackLower, TransformSpace=RTS_ParentBoneSpace))
   > Set FACIAL_L_EyesackLower_Location(FACIAL_L_EyesackLower_Location:Vector=@n5) -> (Output_Get:Vector)
   > Set FACIAL_L_EyesackLower_Rotation(FACIAL_L_EyesackLower_Rotation:Rotator=@n5.Rotation) -> (Output_Get:Rotator)
   n6 = BreakTransform(InTransform=GetSocketTransform(self=Get HeadMesh(), InSocketName=FACIAL_R_EyesackLower, TransformSpace=RTS_ParentBoneSpace))
   > Set FACIAL_R_EyesackLower_Location(FACIAL_R_EyesackLower_Location:Vector=@n6) -> (Output_Get:Vector)
   > Set FACIAL_R_EyesackLower_Rotation(FACIAL_R_EyesackLower_Rotation:Rotator=@n6.Rotation) -> (Output_Get:Rotator)
   n7 = BreakTransform(InTransform=GetSocketTransform(self=Get HeadMesh(), InSocketName=FACIAL_L_EyeCornerOuter, TransformSpace=RTS_ParentBoneSpace))
   > Set FACIAL_L_EyeCornerOuter_Location(FACIAL_L_EyeCornerOuter_Location:Vector=@n7) -> (Output_Get:Vector)
   > Set FACIAL_L_EyeCornerOuter_Rotation(FACIAL_L_EyeCornerOuter_Rotation:Rotator=@n7.Rotation) -> (Output_Get:Rotator)
   n8 = BreakTransform(InTransform=GetSocketTransform(self=Get HeadMesh(), InSocketName=FACIAL_R_EyeCornerOuter, TransformSpace=RTS_ParentBoneSpace))
   > Set FACIAL_R_EyeCornerOuter_Location(FACIAL_R_EyeCornerOuter_Location:Vector=@n8) -> (Output_Get:Vector)
   > Set FACIAL_R_EyeCornerOuter_Rotation(FACIAL_R_EyeCornerOuter_Rotation:Rotator=@n8.Rotation) -> (Output_Get:Rotator)
   n9 = BreakTransform(InTransform=GetSocketTransform(self=Get HeadMesh(), InSocketName=FACIAL_L_CheekInner, TransformSpace=RTS_ParentBoneSpace))
   > Set FACIAL_L_CheekInner_Location(FACIAL_L_CheekInner_Location:Vector=@n9) -> (Output_Get:Vector)
   > Set FACIAL_L_CheekInner_Rotation(FACIAL_L_CheekInner_Rotation:Rotator=@n9.Rotation) -> (Output_Get:Rotator)
   n10 = BreakTransform(InTransform=GetSocketTransform(self=Get HeadMesh(), InSocketName=FACIAL_R_CheekInner, TransformSpace=RTS_ParentBoneSpace))
   > Set FACIAL_R_CheekInner_Location(FACIAL_R_CheekInner_Location:Vector=@n10) -> (Output_Get:Vector)
   > Set FACIAL_R_CheekInner_Rotation(FACIAL_R_CheekInner_Rotation:Rotator=@n10.Rotation) -> (Output_Get:Rotator)
   n11 = BreakTransform(InTransform=GetSocketTransform(self=Get HeadMesh(), InSocketName=FACIAL_L_CheekOuter, TransformSpace=RTS_ParentBoneSpace))
   > Set FACIAL_L_CheekOuter_Location(FACIAL_L_CheekOuter_Location:Vector=@n11) -> (Output_Get:Vector)
   > Set FACIAL_L_CheekOuter_Rotation(FACIAL_L_CheekOuter_Rotation:Rotator=@n11.Rotation) -> (Output_Get:Rotator)
   n12 = BreakTransform(InTransform=GetSocketTransform(self=Get HeadMesh(), InSocketName=FACIAL_R_CheekOuter, TransformSpace=RTS_ParentBoneSpace))
   > Set FACIAL_R_CheekOuter_Location(FACIAL_R_CheekOuter_Location:Vector=@n12) -> (Output_Get:Vector)
   > Set FACIAL_R_CheekOuter_Rotation(FACIAL_R_CheekOuter_Rotation:Rotator=@n12.Rotation) -> (Output_Get:Rotator)
   n13 = BreakTransform(InTransform=GetSocketTransform(self=Get HeadMesh(), InSocketName=FACIAL_L_EyelidLowerB, TransformSpace=RTS_ParentBoneSpace))
   > Set FACIAL_L_EyelidLowerB_Location(FACIAL_L_EyelidLowerB_Location:Vector=@n13) -> (Output_Get:Vector)
   > Set FACIAL_L_EyelidLowerB_Rotation(FACIAL_L_EyelidLowerB_Rotation:Rotator=@n13.Rotation) -> (Output_Get:Rotator)
   n14 = BreakTransform(InTransform=GetSocketTransform(self=Get HeadMesh(), InSocketName=FACIAL_R_EyelidLowerB, TransformSpace=RTS_ParentBoneSpace))
   > Set FACIAL_R_EyelidLowerB_Location(FACIAL_R_EyelidLowerB_Location:Vector=@n14) -> (Output_Get:Vector)
   > Set FACIAL_R_EyelidLowerB_Rotation(FACIAL_R_EyelidLowerB_Rotation:Rotator=@n14.Rotation) -> (Output_Get:Rotator)
   Set Is Targeting Alpha(execute=@n2.then_1, Is Targeting Alpha:double="To Float (Boolean)"(InBool=Get bIsTargeting())) -> (Output_Get:double)
   n15 = Get bIsSkyDiving()
   Set Is Not Skydiving Alpha(execute=@n2.then_2, Is Not Skydiving Alpha:double="To Float (Boolean)"(InBool=NOT Boolean(A=@n15))) -> (Output_Get:double)
  ]}
]
variables [
 bHasFaceBones {type=boolean}
 bIsMovingCrouched {type=boolean}
 FACIAL_L_CheekInner_Location {type=Vector}
 FACIAL_L_CheekInner_Rotation {type=Rotator}
 FACIAL_L_CheekOuter_Location {type=Vector}
 FACIAL_L_CheekOuter_Rotation {type=Rotator}
 FACIAL_L_EyeCornerOuter_Location {type=Vector}
 FACIAL_L_EyeCornerOuter_Rotation {type=Rotator}
 FACIAL_L_EyelidLowerB_Location {type=Vector}
 FACIAL_L_EyelidLowerB_Rotation {type=Rotator}
 FACIAL_L_EyesackLower_Location {type=Vector}
 FACIAL_L_EyesackLower_Rotation {type=Rotator}
 FACIAL_R_CheekInner_Location {type=Vector}
 FACIAL_R_CheekInner_Rotation {type=Rotator}
 FACIAL_R_CheekOuter_Location {type=Vector}
 FACIAL_R_CheekOuter_Rotation {type=Rotator}
 FACIAL_R_EyeCornerOuter_Location {type=Vector}
 FACIAL_R_EyeCornerOuter_Rotation {type=Rotator}
 FACIAL_R_EyelidLowerB_Location {type=Vector}
 FACIAL_R_EyelidLowerB_Rotation {type=Rotator}
 FACIAL_R_EyesackLower_Location {type=Vector}
 FACIAL_R_EyesackLower_Rotation {type=Rotator}
 HeadMesh {type=SkeletalMeshComponent}
 Is Not Skydiving Alpha {type=number}
 Is Targeting Alpha {type=number}
 MovingForwardLeft {type=number}
]
